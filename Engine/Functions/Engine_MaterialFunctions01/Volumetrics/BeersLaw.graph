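# BeersLaw is a volumetrics material function, exposed to the library under Density, that returns exponential falloff through a medium.
# It multiplies a scalar Thickness input, the distance a ray travels, by a Depth Scale input that makes the medium denser.
# It negates that product and uses it as the exponent of a constant approximating e, yielding e^(-d).
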MaterialFunction /Engine/Functions/Engine_MaterialFunctions01/Volumetrics/BeersLaw
graph {
 properties {bExposeToLibrary=true, description="Returns e^(-d) which is a standard exponential density function.", libraryCategoriesText [Density]}
 nodes [
  n1 = FunctionInput() {
   bUsePreviewValueAsDefault=true
   description=The distance a ray travelled through some medium.
   inputName=Thickness
   inputType=FunctionInput_Scalar
   sortPriority=1
  }
  n2 = FunctionInput() {
   bUsePreviewValueAsDefault=true
   description="This is simply a multiplier for the Thicknesss input. You can use this to make some objects more dense, such as rainclouds."
   inputName=Depth Scale
   inputType=FunctionInput_Scalar
   sortPriority=2
   previewValue=<1 0 0 1>
  }
  n3 = Multiply(A=@n1, B=@n2)
  FunctionOutput(None=Power(Base=Constant() {r=2.71}, Exp=Multiply(A=@n3) {constB=-1}))
 ]
}
