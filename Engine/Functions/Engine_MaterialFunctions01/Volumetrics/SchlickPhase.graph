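# Library material function that evaluates the Schlick phase function, adding directional anisotropy to volumetric scattering for effects like fog and water.
# It takes a scalar Cosine, meant as the dot product of two vectors, and an Anisotropy factor k, both defaulting to their preview values.
# A single Custom HLSL node returns (1 - k²) / (4π(1 + k·cosθ)²) as the scalar output.
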MaterialFunction /Engine/Functions/Engine_MaterialFunctions01/Volumetrics/SchlickPhase
graph {
 properties {
  bExposeToLibrary=true
  description=The Schlick Phase Function is useful for adding some directionality or anisotropy to volumetric scattering. Useful for fog/water etc.
 }
 nodes [
  MaterialExpressionFunctionInput_0 = FunctionInput() {
   bUsePreviewValueAsDefault=true
   description=Use a Dot Product between the desired two vectors.
   inputName=Cosine
   inputType=FunctionInput_Scalar
   previewValue=<0.5 0 0 1>
  }
  MaterialExpressionFunctionInput_1 = FunctionInput() {
   bUsePreviewValueAsDefault=true
   description=The amount of anisotropy in the scattering. 0 is purely isotropic and 1 is purely forward scattering.
   inputName=Anisotropy
   inputType=FunctionInput_Scalar
   sortPriority=1
   previewValue=<0.5 0 0 1>
  }
  n1 = Custom!(CosTheta=@MaterialExpressionFunctionInput_0, k=@MaterialExpressionFunctionInput_1) {
   code="\tfloat Inner = (1 + k * CosTheta);\r\n\treturn (1 - k * k) / (4 * PI * Inner * Inner);\r\n\r\n"
   outputType=CMOT_Float1
   inputs [
    {
     inputName=CosTheta
     input {expression=&self:MaterialExpressionFunctionInput_0}
    }
    {inputName=k, input {expression=&self:MaterialExpressionFunctionInput_1}}
   ]
  }
  FunctionOutput(None=@n1)
 ]
}
